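# Sound cue that plays a looping ambient sound for the Biome Scanner quest object in Athena.
# A single looping wave player feeds BiomeScanner_AmbientLoop01 straight to the output under the Fort_Quests sound class at reduced volume.
# Attenuation overrides replace the referenced supply drop marker attenuation asset with a custom falloff and shape extent, limiting how far the loop carries.
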SoundCue /Game/Athena/Sounds/Quests/BiomeScanner/BiomeScanner_AmbientLoop_Cue
root=WavePlayer_2
nodes [{id=WavePlayer_2, properties {bLooping=true, soundWaveAssetPtr=&/BiomeScanner_AmbientLoop01}}]
properties {
 attenuationSettings=/Game/Sounds/Foley_Loot/Containers/SupplyDrops/Fort_SupplyDrop_Marker_Close_Athena_Att
 bOverrideAttenuation=true
 maxDistance=1400
 soundClassObject=&Fort_Quests
 volumeMultiplier=0.7
 attenuationOverrides {falloffDistance=1100, attenuationShapeExtents {x=300}}
}
